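# Physics asset for the SherWolf Gnash male medium head cosmetic mesh, giving its two ears simulation bodies.
# Each ear is a two-bone chain of identical angled capsules, anchored to a small capsule on the head bone.
# Constraints link each tip to its base and each base to the head, sharing the same swing and twist limits on both sides.
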PhysicsAsset /BRCosmetics/Characters/Player/Male/Medium/Heads/M_MED_SherWolf_Gnash_Head/Meshes/M_MED_SherWolf_Gnash_Head_Physics
bodies {
 ear_l_01 [{shapeType=Capsule, length=3, radius=3, shapeName=dyn_hair_01_capsule, center=<-2.0298 0 1.0239>, rotation=<rot 63.2314 0.0001 0>}]
 ear_l_02 [{shapeType=Capsule, length=3, radius=3, shapeName=dyn_hair_01_capsule, center=<-2.0298 0 1.0239>, rotation=<rot 63.2314 0.0001 0>}]
 ear_r_01 [{shapeType=Capsule, length=3, radius=3, shapeName=dyn_hair_01_capsule, center=<-2.0298 0 1.0239>, rotation=<rot 63.2314 -0.0001 0>}]
 ear_r_02 [{shapeType=Capsule, length=3, radius=3, shapeName=dyn_hair_01_capsule, center=<-2.0298 0 1.0239>, rotation=<rot 63.2314 -0.0001 0>}]
 head [{shapeType=Capsule, length=0.0497, radius=2.4035}]
}
constraints [
 {bone1Name=ear_l_01, bone2Name=head, swing1LimitDegrees=60, swing2LimitDegrees=45, twistLimitDegrees=20}
 {bone1Name=ear_l_02, bone2Name=ear_l_01, swing1LimitDegrees=60, swing2LimitDegrees=45, twistLimitDegrees=20}
 {bone1Name=ear_r_02, bone2Name=ear_r_01, swing1LimitDegrees=60, swing2LimitDegrees=45, twistLimitDegrees=20}
 {bone1Name=ear_r_01, bone2Name=head, swing1LimitDegrees=60, swing2LimitDegrees=45, twistLimitDegrees=20}
]
